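# Two-sided window glass material for the Eclipse Hotel, using foliage shading for translucency and built like tinted paper.
# A blurred-noise Blend Mask and PaperTexture reroute support lerps of tiled diffuse and normal samples, scalar ranges and Color-parameter tints.
# A tight camera-to-normal power term, scaled by a reroute and the Color_Time_of_Day function, forms a view-dependent, time-of-day color.
# Unconnected Dojo rice-paper grain samples remain in the graph.
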Material /Game/Environments/Asteria/Sets/EclipseHotel/Materials/M_EclipseHotel_Window_Glass
graph {
 properties {
  bFullyRough=true
  bUsedWithNanite=true
  ditherOpacityMask=true
  pixelDepthOffsetMode=PDOM_Legacy
  shadingModel=MSM_TwoSidedFoliage
  twoSided=true
  usedShadingModels=Two Sided Foliage
 }
 nodes [
  LinearInterpolate(Alpha=NamedRerouteUsage()) {constA=0.2, constB=0.02}
  n1 = TextureCoordinate()
  n2 = Constant() {r=1}
  n3 = TextureCoordinate()
  n4 = Constant() {r=1}
  LinearInterpolate(
   A=TextureSample(UVs=Multiply(A=@n1, B=@n2) {constB=5}) {samplerType=SAMPLERTYPE_Normal, texture=/Game/Environments/Asteria/Sets/EclipseHotel/Textures/T_EclipseHotel_Window_Glass_N}.RGB
   B=TextureSample(UVs=Multiply(A=@n3, B=@n4) {constB=5}) {samplerType=SAMPLERTYPE_Normal, texture=&T_EclipseHotel_Window_Glass_N}.RGB
   Alpha=NamedRerouteUsage()
  )
  LinearInterpolate(Alpha=NamedRerouteUsage()) {constA=0.2, constB=0.4}
  n5 = VectorParameter() {parameterName=Color, defaultValue=<rgba 0.652 0.378 0.238 1>}
  Multiply(A=@n5.RGB, B=LinearInterpolate(Alpha=NamedRerouteUsage()) {constA=0.5})
  n6 = Power(Base=DotProduct(A=CameraVectorWS(), B=PixelNormalWS())) {constExponent=20}
  Multiply(
   A=Multiply(A=LinearInterpolate(Alpha=@n6) {constB=0.1}, B=NamedRerouteUsage())
   B=MaterialFunctionCall() {materialFunction=&Color_Time_of_Day}.Result
  )
  n7 = Multiply(A=TextureCoordinate()) {constB=3}
  NamedRerouteDeclaration(None=Power(Base=TextureSample(UVs=@n7) {samplerType=SAMPLERTYPE_Grayscale, texture=&LowResBlurredNoise}.R)) {name=Blend Mask, nodeColor=<rgba 0 1 0.625 1>}
  n8 = TextureSample(UVs=Multiply(A=TextureCoordinate(), B=Constant() {r=1}) {constB=5}) {samplerSource=SSM_Wrap_WorldGroupSettings, texture=/Game/Environments/Asteria/Sets/EclipseHotel/Textures/T_EclipseHotel_Window_Glass_D}
  n9 = Multiply(A=TextureCoordinate(), B=Constant() {r=1}) {constB=5}
  NamedRerouteDeclaration(None=LinearInterpolate(
    A=@n8.RGB
    B=TextureSample(UVs=@n9) {texture=&T_EclipseHotel_Window_Glass_D}.RGB
    Alpha=NamedRerouteUsage()
   )) {name=PaperTexture, nodeColor=<rgba 0 1 0.25 1>}
  n10 = NamedRerouteUsage()
  Reroute(None=Multiply(A=@n5.RGB, B=LinearInterpolate(Alpha=@n10) {constA=0.6}))
  TextureSample() {samplerSource=SSM_Wrap_WorldGroupSettings, texture=/Game/Environments/Asteria/Sets/Dojo/Materials/RicepaperMaterial/Pattern_PaperGrain}
  TextureSample() {samplerSource=SSM_Wrap_WorldGroupSettings, texture=&Pattern_PaperGrain}
  TextureSample() {samplerSource=SSM_Wrap_WorldGroupSettings, texture=&T_EclipseHotel_Window_Glass_D}
  TextureSample() {
   samplerSource=SSM_Wrap_WorldGroupSettings
   samplerType=SAMPLERTYPE_Normal
   texture=/Game/Environments/Asteria/Sets/Dojo/Materials/RicepaperMaterial/Pattern_PaperGrain_NM
  }
  TextureSample() {
   samplerSource=SSM_Wrap_WorldGroupSettings
   samplerType=SAMPLERTYPE_Normal
   texture=&Pattern_PaperGrain_NM
  }
  VectorParameter() {parameterName=Color, defaultValue=<rgba 0.652 0.378 0.238 1>}
  VectorParameter() {parameterName=Color, defaultValue=<rgba 0.652 0.378 0.238 1>}
  VectorParameter()
 ]
}
